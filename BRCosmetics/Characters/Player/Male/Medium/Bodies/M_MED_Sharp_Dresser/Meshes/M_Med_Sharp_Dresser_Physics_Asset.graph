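# Physics asset for the male medium Sharp Dresser character body, set up around two dynamic tie bone chains.
# Each three-bone capsule chain hangs from spine_05 with limited swing and twist, tighter on the right than the left.
# Capsules cover the pelvis, spine, head, arms and right hand, likely as tie colliders, while leg and foot bodies have no shapes.
# Arm and spine_04 to spine_03 joints are free in rotation; head and upper-spine joints are limited.
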PhysicsAsset /BRCosmetics/Characters/Player/Male/Medium/Bodies/M_MED_Sharp_Dresser/Meshes/M_Med_Sharp_Dresser_Physics_Asset
bodies {
 calf_l []
 calf_r []
 dyn_tie_01_l [{shapeType=Capsule, length=6, radius=1, shapeName=dyn_tie_01_l_capsule, center=<3.5 0 0>, rotation=<rot 90 0 0>}]
 dyn_tie_01_r [{shapeType=Capsule, length=6, radius=1, shapeName=dyn_tie_01_r_capsule, center=<3 0 0>, rotation=<rot 90 0 0>}]
 dyn_tie_02_l [{shapeType=Capsule, length=4, radius=1, shapeName=dyn_tie_02_l_capsule, center=<4 0 0>, rotation=<rot 90 0 0>}]
 dyn_tie_02_r [{shapeType=Capsule, length=2.5, radius=1, shapeName=dyn_tie_02_r_capsule, center=<2.25 0 0>, rotation=<rot 90 0 0>}]
 dyn_tie_03_l [{shapeType=Capsule, length=3, radius=1.5, shapeName=dyn_tie_03_l_capsule, center=<2.54 0 0>, rotation=<rot 90 0 0>}]
 dyn_tie_03_r [{shapeType=Capsule, length=3, radius=1.5, shapeName=dyn_tie_03_r_capsule, center=<3 0 -0.5>, rotation=<rot 90 0 0>}]
 foot_l []
 foot_r []
 hand_r [{shapeType=Capsule, length=9.8479, radius=4.0943, shapeName=hand_r_capsule, center=<-4.2805 0.5544 0>, rotation=<rot 89.3224 -0.0008 -0.0008>}]
 head [{shapeType=Capsule, length=15, radius=11, center=<2 -1 0>, rotation=<rot 90 0 -15>}]
 lowerarm_l [{shapeType=Capsule, length=12, radius=6, shapeName=lowerarm_l_capsule, center=<11 0 0>, rotation=<rot 90 0 0>}]
 lowerarm_r [{shapeType=Capsule, length=12.6319, radius=5.3681, shapeName=lowerarm_r_capsule, center=<-11 0 0>, rotation=<rot 90 0 0>}]
 pelvis [{shapeType=Capsule, length=8.6562, radius=13.5388, center=<-2.6215 -1.3695 0>}]
 spine_01 [{shapeType=Capsule, length=10.7187, radius=12.4138, center=<-2.6215 -0.7156 0>}]
 spine_02 [
  {shapeType=Capsule, length=10.5937, radius=10.945, center=<-2.6215 0.4993 0.0001>}
  {shapeType=Capsule, length=29.875, radius=11.1013, center=<2.5024 0.4992 3.9825>, rotation=<rot 90 0.0049 0.0049>}
  {shapeType=Capsule, length=29.875, radius=11.1013, center=<2.4704 0.4991 -3.9825>, rotation=<rot 90 -0.0396 -0.0396>}
 ]
 spine_03 [{shapeType=Capsule, length=4.5, radius=13.5, center=<3.2338 0 0>}]
 spine_04 [{shapeType=Capsule, length=7, radius=13.5, shapeName=spine_04_capsule, center=<3.7816 -0.4201 0>}]
 spine_05 [{shapeType=Capsule, length=14, radius=11.5, center=<1 0 0>}]
 thigh_l []
 thigh_r []
 upperarm_l [{shapeType=Capsule, length=13, radius=6, shapeName=upperarm_l_capsule, center=<13 0 0>, rotation=<rot 90 0 0>}]
 upperarm_r [{shapeType=Capsule, length=14, radius=6, shapeName=upperarm_r_capsule, center=<-13 0 0>, rotation=<rot 90 0 0>}]
}
constraints [
 {bone1Name=spine_05, bone2Name=spine_04, swing1LimitDegrees=45, swing2LimitDegrees=45, twistLimitDegrees=45}
 {bone1Name=head, bone2Name=spine_05, swing1LimitDegrees=45, swing2LimitDegrees=45, twistLimitDegrees=45}
 {
  bone1Name=spine_04
  bone2Name=spine_03
  swing1LimitDegrees=45
  swing1Motion=Free
  swing2LimitDegrees=45
  swing2Motion=Free
  twistLimitDegrees=45
  twistMotion=Free
 }
 {bone1Name=dyn_tie_03_l, bone2Name=dyn_tie_02_l, swing1LimitDegrees=30, swing2LimitDegrees=30, twistLimitDegrees=5}
 {bone1Name=dyn_tie_01_r, bone2Name=spine_05, swing1LimitDegrees=12, swing2LimitDegrees=15, twistLimitDegrees=1}
 {bone1Name=dyn_tie_02_r, bone2Name=dyn_tie_01_r, swing1LimitDegrees=12, swing2LimitDegrees=15, twistLimitDegrees=1}
 {bone1Name=dyn_tie_03_r, bone2Name=dyn_tie_02_r, swing1LimitDegrees=12, swing2LimitDegrees=15, twistLimitDegrees=1}
 {bone1Name=dyn_tie_01_l, bone2Name=spine_05, swing1LimitDegrees=30, swing2LimitDegrees=30, twistLimitDegrees=5}
 {bone1Name=dyn_tie_02_l, bone2Name=dyn_tie_01_l, swing1LimitDegrees=30, swing2LimitDegrees=30, twistLimitDegrees=5}
 {
  bone1Name=upperarm_l
  bone2Name=spine_05
  swing1LimitDegrees=45
  swing1Motion=Free
  swing2LimitDegrees=45
  swing2Motion=Free
  twistLimitDegrees=45
  twistMotion=Free
 }
 {
  bone1Name=upperarm_r
  bone2Name=spine_05
  swing1LimitDegrees=45
  swing1Motion=Free
  swing2LimitDegrees=45
  swing2Motion=Free
  twistLimitDegrees=45
  twistMotion=Free
 }
 {
  bone1Name=lowerarm_l
  bone2Name=upperarm_l
  swing1LimitDegrees=45
  swing1Motion=Free
  swing2LimitDegrees=45
  swing2Motion=Free
  twistLimitDegrees=45
  twistMotion=Free
 }
 {
  bone1Name=lowerarm_r
  bone2Name=upperarm_r
  swing1LimitDegrees=45
  swing1Motion=Free
  swing2LimitDegrees=45
  swing2Motion=Free
  twistLimitDegrees=45
  twistMotion=Free
 }
 {
  bone1Name=hand_r
  bone2Name=lowerarm_r
  swing1LimitDegrees=45
  swing1Motion=Free
  swing2LimitDegrees=45
  swing2Motion=Free
  twistLimitDegrees=45
  twistMotion=Free
 }
]
