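# Animation Blueprint for the Sharp Dresser medium male body, adding secondary motion and corrective offsets to a copied owner-mesh pose.
# RigidBody and speed-weighted hood AnimDynamics feed a cached pose.
# That pose gets near-zero head scale when the playlist uses custom character parts, then head offset, Control Rig and glute/edge bone corrections.
# The event graph smoothly rolls the backpack component from spine twist, speed and jump/fall state.
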AnimBlueprint /BRCosmetics/Characters/Player/Male/Medium/Bodies/M_MED_Sharp_Dresser/Meshes/M_MED_Sharp_Dresser_AnimBP
target_skeleton=&/M_MED_Sharp_Dresser_Skeleton
graphs [
 AnimGraph {nodes [
   n1 = Anim Dynamics - Bone: dyn_hood_l(Alpha=1, ExternalForce="0,0,0") {
    angularSpringConstant=1600
    bGravityOverrideInSimSpace=true
    bLinearSpring=true
    alphaBoolBlend {blendOption=Linear}
    boundBone {boneName=dyn_hood_l}
    physicsBodyDefinitions [{
      sphereCollisionRadius=0
      boundBone {boneName=dyn_hood_l}
      boxExtents=<15 15 15>
      constraintSetup {angularLimitsMax=<0 0 0>, angularLimitsMin=<0 0 0>, angularTarget=<1 0 0>, linearAxesMax=<2 2 2>, linearAxesMin=<-2 0 -2>}
      localJointOffset=<0 0 0>
     }]
   }
   n2 = Anim Dynamics - Bone: dyn_hood_r(ComponentPose=@n1, Alpha=1, ExternalForce="0,0,0") {
    angularSpringConstant=1600
    bGravityOverrideInSimSpace=true
    bLinearSpring=true
    alphaBoolBlend {blendOption=Linear}
    boundBone {boneName=dyn_hood_r}
    physicsBodyDefinitions [{
      sphereCollisionRadius=0
      boundBone {boneName=dyn_hood_r}
      boxExtents=<15 15 15>
      constraintSetup {angularLimitsMax=<0 0 0>, angularLimitsMin=<0 0 0>, angularTarget=<1 0 0>, linearAxesMax=<2 0 2>, linearAxesMin=<-2 -2 -2>}
      localJointOffset=<0 0 0>
     }]
   }
   Anim Dynamics - Bone: dyn_hood_c(ComponentPose=@n2, Alpha=1, ExternalForce="0,0,0") {
    angularSpringConstant=1600
    bGravityOverrideInSimSpace=true
    bLinearSpring=true
    alphaBoolBlend {blendOption=Linear}
    boundBone {boneName=dyn_hood_c}
    physicsBodyDefinitions [{
      sphereCollisionRadius=0
      boundBone {boneName=dyn_hood_c}
      boxExtents=<15 15 15>
      constraintSetup {angularLimitsMax=<0 0 0>, angularLimitsMin=<0 0 0>, angularTarget=<1 0 0>, linearAxesMax=<2 2 2>, linearAxesMin=<0 -2 -2>}
      localJointOffset=<0 0 0>
     }]
   }
   n3 = "Transform (Modify) Bone - Bone: head"!(ComponentPose=Local To Component(LocalPose=Use cached pose 'PreScale Cache'()), Alpha=1) -> (Pose) {scaleMode=BMM_Replace, alphaBoolBlend {blendOption=Linear}, boneToModify {boneName=head}, scale=<0.01 0.01 0.01>}
   n4 = Component To Local(ComponentPose=@n3)
   n5 = Use cached pose 'PreScale Cache'()
   n6 = Get GetPlayerListUsesCustomCharacterParts() -> (GetPlayerListUsesCustomCharacterParts:bool)
   n7 = "Transform (Modify) Bone - Bone: head"!(
    ComponentPose=Local To Component(LocalPose=Blend Poses by bool(BlendPose_0=@n4, BlendPose_1=@n5, BlendTime_0=0, BlendTime_1=0, bActiveValue=@n6) {blendType=Linear})
    Alpha=1
    Translation:Vector="1.1,0,0"
    Rotation:Rotator="0,0,0"
    Scale:Vector="1,1,1"
   ) -> (Pose) {
    rotationMode=BMM_Additive
    translationMode=BMM_Additive
    translationSpace=BCS_BoneSpace
    alphaBoolBlend {blendOption=Linear}
    boneToModify {boneName=head}
    translation {x=1.1}
   }
   n8 = Control Rig(Source=Component To Local(ComponentPose=@n7)) {
    alphaBoolBlend {blendOption=Linear}
    controlRigAssetReference {blueprintRigClass=&/M_MED_Sharp_Dresser_CtrlRig.M_MED_Sharp_Dresser_CtrlRig_C}
   }
   n9 = "Transform (Modify) Bone - Bone: deform_glute_base_l"(ComponentPose=Local To Component(LocalPose=@n8), Alpha=1, Translation="2.649,1.359,-0.000", Rotation="0.000,-14.080,0.000", Scale="1,1,1") {
    rotationMode=BMM_Additive
    rotationSpace=BCS_BoneSpace
    translationMode=BMM_Additive
    translationSpace=BCS_BoneSpace
    alphaBoolBlend {blendOption=Linear}
    boneToModify {boneName=deform_glute_base_l}
    rotation=<rot 0 -14.0801 0>
    translation=<2.649 1.359>
   }
   n10 = "Transform (Modify) Bone - Bone: deform_glute_base_r"(ComponentPose=@n9, Alpha=1, Translation="-0.746,-0.585,0", Rotation="0.000,-14.077,0.000", Scale="1,1,1") {
    rotationMode=BMM_Additive
    rotationSpace=BCS_BoneSpace
    translationMode=BMM_Additive
    translationSpace=BCS_BoneSpace
    alphaBoolBlend {blendOption=Linear}
    alphaScaleBiasClamp {bInterpResult=true}
    boneToModify {boneName=deform_glute_base_r}
    rotation=<rot 0 -14.0771 0>
    translation=<-0.746 -0.585>
   }
   n11 = "Transform (Modify) Bone - Bone: dyn_edgecorrectiveup"(ComponentPose=@n10, Alpha=1, Translation="0.000,5.861,0.441", Rotation="0.000,-0.000,-0.010", Scale="1,1,1") {
    rotationMode=BMM_Additive
    rotationSpace=BCS_BoneSpace
    translationMode=BMM_Additive
    translationSpace=BCS_BoneSpace
    alphaBoolBlend {blendOption=Linear}
    alphaScaleBiasClamp {bInterpResult=true}
    boneToModify {boneName=dyn_edgecorrectiveup}
    rotation=<rot 0 0 -0.0101>
    translation=<0 5.8613 0.441>
   }
   Output Pose(Result=Component To Local(ComponentPose=@n11)) {layerGroup=None}
   n12 = Get MeshToCopy() -> (MeshToCopy:Skeletal Mesh Component Object Reference)
   n13 = RigidBody(ComponentPose=Local To Component(LocalPose=Copy Pose From Mesh(SourceMeshComponent=@n12))) {
    bClampLinearTranslationLimitToRefPose=true
    bOverrideWorldGravity=true
    overridePhysicsAsset=&/M_Med_Sharp_Dresser_Physics_Asset
    simulationSpace=BaseBoneSpace
    alphaBoolBlend {blendOption=Linear}
    baseBoneRef {boneName=spine_05}
    componentAppliedLinearAccClamp=<3000 3000 3000>
    componentLinearAccScale=<0.3 0.3 0.5>
    componentLinearVelScale=<0.7 0.7 0.8>
    overrideWorldGravity {z=-980}
    simSpaceSettings {maxAngularAcceleration=100, maxAngularVelocity=1000, maxLinearAcceleration=100, maxLinearVelocity=1000, worldAlpha=0.5}
   }
   n14 = MapRangeClamped(Value=Get PawnSpeed2D(), InRangeA=0, InRangeB=100, OutRangeA=0.1)
   n15 = Anim Dynamics - Bone: dyn_hood_l(ComponentPose=@n13, Alpha=@n14, ExternalForce="0,0,0") {
    angularDampingOverride=0.9
    angularSpringConstant=15
    bAngularSpring=true
    bGravityOverrideInSimSpace=true
    simulationSpace=BoneRelative
    alphaBoolBlend {blendOption=Linear}
    boundBone {boneName=dyn_hood_l}
    componentLinearAccScale=<1 1 1>
    componentLinearVelScale=<1 1 1>
    physicsBodyDefinitions [{
      sphereCollisionRadius=0
      boundBone {boneName=dyn_hood_l}
      boxExtents=<10 20 20>
      constraintSetup {angularLimitsMax=<30 30 30>, angularLimitsMin=<-30 -30 -30>, angularTarget=<1 0 0>}
      localJointOffset=<0 0 0>
     }]
   }
   n16 = Anim Dynamics - Bone: dyn_hood_r(ComponentPose=@n15, Alpha=@n14, ExternalForce="0,0,0") {
    angularDampingOverride=0.9
    angularSpringConstant=15
    bAngularSpring=true
    bGravityOverrideInSimSpace=true
    simulationSpace=BoneRelative
    alphaBoolBlend {blendOption=Linear}
    boundBone {boneName=dyn_hood_r}
    componentLinearAccScale=<1 1 1>
    componentLinearVelScale=<1 1 1>
    physicsBodyDefinitions [{
      sphereCollisionRadius=0
      boundBone {boneName=dyn_hood_r}
      boxExtents=<10 20 20>
      constraintSetup {angularLimitsMax=<30 30 30>, angularLimitsMin=<-30 -30 -30>, angularTarget=<1 0 0>}
      localJointOffset=<0 0 0>
     }]
   }
   n17 = Anim Dynamics - Bone: dyn_hood_c(ComponentPose=@n16, Alpha=@n14, ExternalForce="0,0,0") {
    angularDampingOverride=0.9
    angularSpringConstant=15
    bAngularSpring=true
    bGravityOverrideInSimSpace=true
    simulationSpace=BoneRelative
    alphaBoolBlend {blendOption=Linear}
    boundBone {boneName=dyn_hood_c}
    componentLinearAccScale=<1 1 1>
    componentLinearVelScale=<1 1 1>
    physicsBodyDefinitions [{
      sphereCollisionRadius=0
      boundBone {boneName=dyn_hood_c}
      boxExtents=<10 20 20>
      constraintSetup {angularLimitsMax=<30 30 30>, angularLimitsMin=<-30 -30 -30>, angularTarget=<1 0 0>}
      localJointOffset=<0 0 0>
     }]
   }
   Save cached pose 'PreScale Cache'(Pose=Component To Local(ComponentPose=@n17)) {cacheName=PreScale Cache}
  ]}
 EventGraph {nodes [
   Event BlueprintUpdateAnimation()
   > Set LOD_alpha(LOD_alpha:bool="Equal (Integer)"(A=Get CurrentLOD(), B=3)) -> (Output_Get:bool)
   n1 = MapRangeClamped(Value=Vector Length(A=Get PawnVelocity()), InRangeA=0, InRangeB=350, OutRangeA=0.1)
   n2 = GetOwnerSkeletalMeshComponent()
   n3 = "Absolute (Float)"(A=MapRangeClamped(
     Value=GetDeltaTransformFromRefPose!(self=@n2, BoneName=spine_05, BaseName=pelvis) -> (
      ReturnValue_Location:Vector
      ReturnValue_Rotation_Roll:float
      ReturnValue_Rotation_Pitch:float
      ReturnValue_Rotation_Yaw:float
      ReturnValue_Scale:Vector
     ).ReturnValue_Rotation_Roll
     InRangeA=-25
     InRangeB=25
     OutRangeA=-1
    ))
   n4 = OR Boolean(A=Get bIsJumping(), B=Get bIsFalling())
   > n5 = Set BackpackSocket_alpha(BackpackSocket_alpha:double=FInterpTo(
     Current=Get BackpackSocket_alpha() -> (BackpackSocket_alpha:double)
     Target=float + float!(A:bool=@n4, B=@n3, C:double=@n1) -> (ReturnValue)
     DeltaTime=GetWorldDeltaSeconds!() -> (ReturnValue:double)
     InterpSpeed=2
    )) -> (Output_Get:double)
   n6 = Get Output() -> (Output:Skeletal Mesh Component Budgeted Object Reference)
   n7 = Get!(self:Scene Component Object Reference=@n6) -> (RelativeLocation_X:double, RelativeLocation_Y:double, RelativeLocation_Z:double)
   n8 = Get!(self=@n6) -> (RelativeRotation_Roll:float, RelativeRotation_Pitch:float, RelativeRotation_Yaw:float)
   > Set Relative Location And Rotation!(
    execute
    self:Scene Component Object Reference=@n6
    NewLocation_X:double=@n7
    NewLocation_Y:double=@n7.RelativeLocation_Y
    NewLocation_Z:double=@n7.RelativeLocation_Z
    NewRotation_Roll:float=MapRangeClamped(Value=@n5, InRangeA=0, InRangeB=1, OutRangeB=15)
    NewRotation_Pitch:float=@n8.RelativeRotation_Pitch
    NewRotation_Yaw:float=@n8.RelativeRotation_Yaw
    bSweep:bool=false
    bTeleport:bool=false
   ) -> (then, SweepHitResult:Hit Result Structure) {is=title K2Node_CallFunction}
   n9 = Event OnCharacterCustomizationLoaded()
   > Set Output(Output:Skeletal Mesh Component Budgeted Object Reference=Get Array Item with EFort Custom Part Type(Array=@n9.CustomMeshParts, PartType=Backpack)) -> (Output_Get:Skeletal Mesh Component Budgeted Object Reference)
   Event BlueprintInitializeAnimation()
   > Set MeshToCopy(MeshToCopy:Skeletal Mesh Component Object Reference=GetOwnerSkeletalMeshComponent()) -> (Output_Get:Skeletal Mesh Component Object Reference)
   > Set GetPlayerListUsesCustomCharacterParts(GetPlayerListUsesCustomCharacterParts:bool=GetPlaylistUsesCustomCharacterParts!() -> (ReturnValue:bool)) -> (Output_Get:bool)
   "To Float (Boolean)"(InBool=@n4)
  ]}
]
variables [
 BackpackSocket_alpha {type=number}
 GetPlayerListUsesCustomCharacterParts {type=boolean}
 LOD_alpha {type=boolean}
 MeshToCopy {type=SkeletalMeshComponent}
 Output {type=SkeletalMeshComponentBudgeted}
]
